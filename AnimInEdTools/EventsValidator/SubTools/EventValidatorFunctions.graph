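# Widget-less Editor Utility Widget holding validation functions that check animation assets against per-project event specifications.
# It loads the DataTable matching the project name, picks a row from the asset name, and matches character keys for frame ranges.
# Anim sequences are checked for length (half the table range), 30 FPS and Fortnite bone compression; sections for start/end frames.
# Level sequences fail on Control Rig sections; WIP sequences compare playback range to halved input bounds.
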EditorUtilityWidgetBlueprint /AnimInEdTools/EventsValidator/SubTools/EventValidatorFunctions
graphs [
 EventGraph {links [], nodes []}
 GetDataTable {nodes [
   n1 = GetDataTable!() -> (then, Project:BP Animation Browser Project Object Reference) {is=title K2Node_FunctionEntry}
   > Set Project(Project:BP Animation Browser Project Object Reference=@n1) -> (Output_Get:BP Animation Browser Project Object Reference)
   > n2 = GetAssetsByPath!(
    execute
    self:Asset Registry Interface=GetAssetRegistry()
    PackagePath:Name=String To Name(InString=/AnimInEdTools/EventsValidator/ProjectDataTables)
    bRecursive:bool=false
    bIncludeOnlyOnDiskAssets:bool=false
   ) -> (then, OutAssetData:Array of Asset Data Structures, ReturnValue:bool)
   > n3 = For Each Loop!(Exec, Array:Array of Asset Data Structures=@n2) -> (
    LoopBody
    Array Element_PackageName:Name
    Array Element_PackagePath:Name
    Array Element_AssetName:Name
    Array Element_AssetClassPath:Top Level Asset Path Structure
    Array Index
    Completed
   )
   n4 = "To String (Name)"(InName=@n3.Array Element_AssetName)
   n5 = Get Project() -> (Project:BP Animation Browser Project Object Reference)
   > Branch(Condition=Contains(SearchIn=@n4, Substring=Get Name!(self:BP Animation Browser Project Object Reference=@n5) -> (Name:String)))
   > n6 = LoadAsset(AssetPath=Append!(A="To String (Name)"(InName=@n3), B=., C:String=@n4) -> (ReturnValue))
   > n7 = Cast To DataTable!(execute, Object=@n6) -> (then, CastFailed, AsData Table:Data Table Object Reference) {is=title K2Node_DynamicCast}
   > Set ProjectDataTable(ProjectDataTable:Data Table Object Reference=@n7) -> (Output_Get:Data Table Object Reference)
  ]}
 GetTableRow {nodes [
   n1 = GetTableRow!() -> (then, AssetName:String) {is=title K2Node_FunctionEntry}
   n2 = Get Project()
   n3 = Replace(SourceString=@n1, From=Append!(A=Get Name(self=@n2), B=_) -> (ReturnValue))
   n4 = FindSubstring(SearchIn=@n3, Substring=_)
   > n5 = Get Data Table Row!(
    execute:Exec
    DataTable:Data Table Object Reference=Get ProjectDataTable() -> (ProjectDataTable:Data Table Object Reference)
    RowName:Name=String To Name(InString=Left!(SourceString:String=@n3, Count:int=@n4) -> (ReturnValue:String))
   ) -> (then:Exec, RowNotFound:Exec, ReturnValue:Struct Events Validator Structure)
   > Set CurrentTableRow(CurrentTableRow:Struct Events Validator Structure=@n5) -> (Output_Get:Struct Events Validator Structure)
   > n6 = Set Found(Found:bool=true) -> (Output_Get:bool)
   Return Node!(execute=[@n6.then, @n7.then], RowFound:bool=Get Found() -> (Found:bool))
   n7 = Set Found(execute=@n5.RowNotFound, Found=false)
   PrintString(InString=TABLE ROW NOT FOUND, TextColor=(R=1,G=0.0395,B=0,A=1))
  ]}
 Validate {nodes [
   n1 = Validate!() -> (then, Project:BP Animation Browser Project Object Reference, AssetToValidate:Object Reference) {is=title K2Node_FunctionEntry}
   > Set Project(Project=@n1)
  ]}
 ValidateAnimSeq {nodes [
   n1 = ValidateAnimSeq!() -> (then, Project:BP Animation Browser Project Object Reference, AnimSequence:Anim Sequence Object Reference) {is=title K2Node_FunctionEntry}
   > n2 = Is Valid(InputObject=@n1.AnimSequence)
   PrintString(execute=@n2.Is Not Valid, InString=ANIM SEQ NOT VALID)
   GetDataTable!(execute=@n2.Is Valid, self, Project:BP Animation Browser Project Object Reference=Get Project()) -> (then) {is=title K2Node_CallFunction}
   > GetTableRow!(execute, self, AssetName:String=GetDisplayName(Object=@n1.AnimSequence)) -> (then, RowFound:bool) {is=title K2Node_CallFunction}
   n3 = Break Struct Events Validator!(Struct_EventsValidator:"Struct Events Validator Structure (by ref)"=Get CurrentTableRow() -> (CurrentTableRow:Struct Events Validator Structure)) -> (Character_7_#1:Map of Strings to Struct Events Validator Frame Range Structures) {is=title K2Node_BreakStruct}
   > n4 = Keys!(execute, TargetMap:Map of Strings to Struct Events Validator Frame Range Structures=@n3) -> (then, Keys:Array of Strings)
   > n5 = For Each Loop!(Exec, Array:Array of Strings=@n4) -> (LoopBody, Array Element:String, Array Index, Completed)
   n6 = Branch(execute=@n5.Completed, Condition=Get Valid() -> (Valid:bool))
   > n7 = Set AnimSatusText(AnimSatusText:String=Anim sequence is valid) -> (Output_Get:String)
   Return Node!(
    execute=[@n13.then, @n12.then, @n7.then]
    IsValid:bool=Get Valid()
    RowFound=Get isRowFound() -> (isRowFound:bool)
    RequiredStartFrame:int=Get Start Frame() -> (Start Frame:int=0)
    RequiredEndFrame:int=Get End Frame() -> (End Frame:int=0)
    StatusText:String=Get AnimSatusText() -> (AnimSatusText:String)
   )
   Set AnimSatusText(execute=@n6.else, AnimSatusText=Anim sequence is not valid:)
   > n8 = Branch(Condition=Get isRowFound())
   > n9 = Set AnimSatusText(AnimSatusText=Append(A=Get AnimSatusText(), B=" [Animation length]"))
   n10 = Branch(execute=[@n15.then, @n9.then], Condition=Get Is30FPS() -> (Is30FPS:bool))
   n11 = Get AnimSequence() -> (AnimSequence:Anim Sequence Object Reference)
   n12 = Branch(
    execute=[@n10.then, @n14.then]
    Condition="Equal (Object)"(
     A=GetBoneCompressionSettings!(AnimationSequence:Anim Sequence Object Reference=@n11) -> (CompressionSettings:Anim Bone Compression Settings Object Reference)
     B=/Game/Animation/Settings/FortniteAnimBoneCompressionSettings
    )
   )
   n13 = Set AnimSatusText(execute=@n12.else, AnimSatusText=Append(A=Get AnimSatusText(), B=" [Bone Compression Setting] "))
   n14 = Set AnimSatusText(execute=@n10.else, AnimSatusText=Append(A=Get AnimSatusText(), B=" [Not 30 FPS] "))
   n15 = Set AnimSatusText(execute=@n8.else, AnimSatusText=Append(A=Get AnimSatusText(), B=" [Character missing in data table] "))
   n16 = Contains(SearchIn=GetDisplayName(Object=Get AnimSequence()), Substring=@n5)
   Branch(execute=@n5.LoopBody, Condition=@n16)
   > Set isRowFound(isRowFound:bool=@n16) -> (Output_Get:bool)
   n17 = Break Struct Events Validator Frame Range!(Struct_EventsValidatorFrameRange:"Struct Events Validator Frame Range Structure (by ref)"=Find(TargetMap:Map of Strings to Struct Events Validator Frame Range Structures=@n3, Key=@n5) -> (Value:Struct Events Validator Frame Range Structure)) -> (StartFrame_2_#2:int, EndFrame_4_#3:int) {is=title K2Node_BreakStruct}
   > Set Start Frame(Start Frame:int=@n17) -> (Output_Get:int=0)
   > Set End Frame(End Frame:int=@n17.EndFrame_4_#3) -> (Output_Get:int=0)
   n18 = Get AnimSequence()
   n19 = GetNumFrames!(AnimationSequenceBase:Anim Sequence Base Object Reference=@n18) -> (NumFrames:int=0)
   n20 = GetSequenceLength!(AnimationSequenceBase:Anim Sequence Base Object Reference=@n18) -> (Length:float)
   n21 = "Equal (Integer)"(A=Round(A=float / float(A:int=@n19, B=@n20)) {is=title K2Node_CallFunction}, B=30)
   > Set Is30FPS(Is30FPS:bool=@n21) -> (Output_Get:bool)
   n22 = int / int(A=int - int(A=@n17.EndFrame_4_#3, B=@n17), B=2)
   n23 = GetBoneCompressionSettings(AnimationSequence=Get AnimSequence())
   > Set Valid(Valid:bool=AND Boolean!(A="Equal (Integer)"(A=@n19, B=@n22), B=@n21, C:bool="Equal (Object)"(A=@n23, B=&FortniteAnimBoneCompressionSettings)) -> (ReturnValue)) -> (Output_Get:bool)
  ]}
 ValidateAnimSeqSection {nodes [
   n1 = ValidateAnimSeqSection!() -> (
    then
    Project:BP Animation Browser Project Object Reference
    Section:Movie Scene Skeletal Animation Section Object Reference
    AnimSequence:Object Reference
   ) {is=title K2Node_FunctionEntry}
   > GetDataTable(Project=Get Project())
   > GetTableRow(AssetName=GetDisplayName(Object=@n1.AnimSequence))
   > n2 = Cast To AnimSequence!(execute, Object=Get AnimSequence() -> (AnimSequence:Object Reference)) -> (then, CastFailed, AsAnim Sequence:Anim Sequence Object Reference)
   > Set AnimSeqRef(AnimSeqRef:Anim Sequence Object Reference=@n2) -> (Output_Get:Anim Sequence Object Reference)
   n3 = Get!() -> (CurrentTableRow_Character_7_#1:Map of Strings to Struct Events Validator Frame Range Structures)
   > n4 = Keys(TargetMap=@n3)
   > n5 = For Each Loop(Array=@n4)
   n6 = Get AnimSequence()
   > Branch(Condition=Contains(SearchIn=GetDisplayName(Object=@n6), Substring=@n5))
   n7 = Get Section() -> (Section:Movie Scene Skeletal Animation Section Object Reference)
   > n8 = GetStartFrame!(execute, Section:Movie Scene Section Object Reference=@n7) -> (then, ReturnValue:int=0)
   > n9 = GetEndFrame!(execute, Section:Movie Scene Section Object Reference=@n7) -> (then, ReturnValue:int=0)
   n10 = Find!(TargetMap=@n3, Key=@n5) -> (Value_StartFrame_2_#2:int, Value_EndFrame_4_#3:int, ReturnValue)
   n11 = "Equal (Integer)"(A=@n8, B=@n10)
   n12 = "Equal (Integer)"(A=@n9, B=@n10.Value_EndFrame_4_#3)
   n13 = AND Boolean!(A=@n11, B=@n12) -> (ReturnValue)
   > n14 = Branch(Condition=@n13)
   > n15 = Set AnimSectionStatus(AnimSectionStatus:String=Section is valid) -> (Output_Get:String)
   Return Node!(execute=[@n17.then, @n15.then, @n18.then], IsValid=@n13, StatusText=Get AnimSectionStatus() -> (AnimSectionStatus:String))
   Set AnimSectionStatus(execute=@n14.else, AnimSectionStatus=Section is not valid:)
   > n16 = Branch(Condition=@n11)
   n17 = Branch(execute=[@n16.then, @n19.then], Condition=@n12)
   n18 = Set AnimSectionStatus(execute=@n17.else, AnimSectionStatus=Append(A=Get AnimSectionStatus(), B=" [End Frame] "))
   n19 = Set AnimSectionStatus(execute=@n16.else, AnimSectionStatus=Append(A=Get AnimSectionStatus(), B=" [Start Frame] "))
  ]}
 ValidateLvlSeq {nodes [
   n1 = ValidateLvlSeq!() -> (then, Project:BP Animation Browser Project Object Reference, LevelSequence:Level Sequence Object Reference) {is=title K2Node_FunctionEntry}
   > Is Valid(InputObject=@n1.LevelSequence)
   > n2 = GetBindings!(execute, Sequence:Movie Scene Sequence Object Reference=Get LevelSequence() -> (LevelSequence:Level Sequence Object Reference)) -> (then, ReturnValue:Array of Movie Scene Binding Proxy Structures)
   > n3 = For Each Loop(Array:Array of Movie Scene Binding Proxy Structures=@n2) -> (Array Element:Movie Scene Binding Proxy Structure)
   n4 = Branch(execute=@n3.Completed, Condition=Get HasControlRig() -> (HasControlRig:bool))
   > n5 = Set SeqStatusText(SeqStatusText:String="Sequence is not valid: [Has control rig]") -> (Output_Get:String)
   Return Node(execute=[@n6.then, @n5.then], IsValid=Get Valid(), StatusText=Get SeqStatusText() -> (SeqStatusText:String))
   n6 = Set Valid(execute=@n4.else, Valid=true)
   n7 = FindTracksByExactType!(
    execute=@n3.LoopBody
    InBinding:"Movie Scene Binding Proxy Structure (by ref)"=@n3
    TrackType:Movie Scene Track Class Reference=/Script/ControlRig.MovieSceneControlRigParameterTrack
   ) -> (then, ReturnValue:Array of Movie Scene Control Rig Parameter Track Object References)
   > n8 = For Each Loop(Array:Array of Movie Scene Control Rig Parameter Track Object References=@n7) -> (Array Element:Movie Scene Control Rig Parameter Track Object Reference)
   > n9 = GetSections!(execute, Track:Movie Scene Track Object Reference=@n8) -> (then, ReturnValue:Array of Movie Scene Section Object References)
   > n10 = For Each Loop(Array:Array of Movie Scene Section Object References=@n9) -> (Array Element:Movie Scene Section Object Reference)
   > n11 = Cast To MovieSceneControlRigParameterSection!(execute, Object=@n10) -> (then, CastFailed, AsMovie Scene Control Rig Parameter Section:Movie Scene Control Rig Parameter Section Object Reference) {is=title K2Node_DynamicCast}
   > Branch(Condition=IsValid(Object=@n11))
   > Do Once()
   > Set HasControlRig(HasControlRig:bool=true) -> (Output_Get:bool)
  ]}
 ValidateWIPSeq {nodes [
   n1 = ValidateWIPSeq!() -> (then, Project:BP Animation Browser Project Object Reference, LevelSequence:Level Sequence Object Reference, Range Start:int, Range End:int) {is=title K2Node_FunctionEntry}
   > n2 = Is Valid(InputObject=@n1.LevelSequence)
   PrintString(execute=@n2.Is Not Valid, InString=LVL SEQ NOT VALID)
   GetDataTable(execute=@n2.Is Valid, Project=Get Project())
   > n3 = GetTableRow(AssetName=GetDisplayName(Object=Get LevelSequence()))
   > n4 = Set IsPartOfProject(IsPartOfProject:bool=@n3) -> (Output_Get:bool)
   > Branch(Condition=@n4)
   > n5 = GetPlaybackRange!(execute, Sequence:Movie Scene Sequence Object Reference=Get LevelSequence()) -> (then, ReturnValue_bHasStart:bool=~, ReturnValue_bHasEnd:bool=~, ReturnValue_InclusiveStart:int, ReturnValue_ExclusiveEnd:int)
   n6 = Get Range Start() -> (Range Start:int=0)
   n7 = Round(A=float / float(A:double="To Float (Integer)"(InInt=@n6), B=2))
   > Set StartFrame(StartFrame:int=@n7) -> (Output_Get:int=0)
   n8 = Get Range End() -> (Range End:int=0)
   n9 = Round(A=float / float(A="To Float (Integer)"(InInt=@n8), B=2))
   > Set EndFrame(EndFrame:int=@n9) -> (Output_Get:int=0)
   n10 = Get StartFrame() -> (StartFrame:int=0)
   n11 = Get EndFrame() -> (EndFrame:int=0)
   > PrintString(InString=Append!(A="To String (Integer)"(InInt=@n10), B=" : ", C:String="To String (Integer)"(InInt=@n11)) -> (ReturnValue))
   n12 = "Equal (Integer)"(A=@n5.ReturnValue_InclusiveStart, B=@n7)
   n13 = "Equal (Integer)"(A=@n5.ReturnValue_ExclusiveEnd, B=@n9)
   > n14 = Set Valid(Valid=AND Boolean(A=@n12, B=@n13))
   > n15 = Branch(Condition=@n14)
   > n16 = Set SeqStatusText(SeqStatusText=Sequence is valid)
   Return Node!(
    execute=[@n18.then, @n19.then, @n16.then]
    PartOfProject:bool=Get IsPartOfProject() -> (IsPartOfProject:bool)
    IsValid=Get Valid()
    StatusText=Get SeqStatusText()
   )
   Set SeqStatusText(execute=@n15.else, SeqStatusText=Sequence is not valid:)
   > n17 = Branch(Condition=@n12)
   n18 = Branch(execute=[@n17.then, @n20.then], Condition=@n13)
   n19 = Set SeqStatusText(execute=@n18.else, SeqStatusText=Append!(A=Get SeqStatusText(), B=" [End Frame] ") -> (ReturnValue))
   n20 = Set SeqStatusText(execute=@n17.else, SeqStatusText=Append(A=Get SeqStatusText(), B=" [Start Frame] "))
  ]}
]
variables [
 CurrentTableRow {type=Struct_EventsValidator, default {character {}}}
 Project {type=BP_AnimationBrowser_Project_C, default=null}
 ProjectDataTable {type=DataTable, default=null}
]
widgets {info {rootWidgetClass=null, widgetCount=0}, widgets []}
